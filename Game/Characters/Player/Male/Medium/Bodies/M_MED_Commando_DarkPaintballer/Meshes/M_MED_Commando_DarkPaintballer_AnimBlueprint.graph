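# Post-copy Anim Blueprint for the Dark Paintballer male commando body, adding secondary motion to the pose copied from its parent mesh.
# Weighted spine and pelvis constraints drive both butt flaps, then Anim Dynamics simulates the flaps and three spring-based scarf bones.
# The cached result gets its head bone shrunk when the playlist uses custom character parts, then passes through Control Rig.
# The event graph also derives movementAnimDyn from pawn speed, unused here.
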AnimBlueprint /Game/Characters/Player/Male/Medium/Bodies/M_MED_Commando_DarkPaintballer/Meshes/M_MED_Commando_DarkPaintballer_AnimBlueprint
target_skeleton=&/M_MED_Commando_DarkPaintballer_Skeleton
graphs [
 AnimGraph {nodes [
   n1 = "Transform (Modify) Bone - Bone: head"!(ComponentPose=Local To Component(LocalPose=Use cached pose 'PreScale Cache'()), Alpha=1) -> (Pose) {scaleMode=BMM_Replace, alphaBoolBlend {blendOption=Linear}, boneToModify {boneName=head}, scale=<0.01 0.01 0.01>}
   n2 = Control Rig(Source=Blend Poses by bool(
     BlendPose_0=Component To Local(ComponentPose=@n1)
     BlendPose_1=Use cached pose 'PreScale Cache'()
     BlendTime_0=0
     BlendTime_1=0
     bActiveValue=Get GetPlayerListUsesCustomCharacterParts() -> (GetPlayerListUsesCustomCharacterParts:bool)
    ) {blendType=Linear}) {
    alphaBoolBlend {blendOption=Linear}
    controlRigAssetReference {blueprintRigClass=&/M_MED_Commando_DarkPaintballer_CtrlRig.M_MED_Commando_DarkPaintballer_CtrlRig_C}
   }
   Output Pose(Result=@n2) {layerGroup=None}
   n3 = "Constraint - dyn_butt_flap_l (PPPP)"!(
    ComponentPose=Local To Component(LocalPose=Copy Pose From Mesh() {bUseAttachedParent=true})
    Alpha=1
    ConstraintWeights_0=0.0354
    ConstraintWeights_1:float=0.003
    ConstraintWeights_2:float=0.7181
    ConstraintWeights_3:float=0.2435
   ) -> (Pose) {
    alphaBoolBlend {blendOption=Linear}
    boneToModify {boneName=dyn_butt_flap_l}
    constraintSetup [{targetBone {boneName=spine_03}}, {targetBone {boneName=spine_02}}, {targetBone {boneName=pelvis}}, {targetBone {boneName=spine_01}}]
    constraintWeights [1,1,1,1]
   }
   n4 = "Constraint - dyn_butt_flap_r (PPPP)"!(
    ComponentPose=@n3
    Alpha=1
    ConstraintWeights_0=0.0355
    ConstraintWeights_1:float=0.0031
    ConstraintWeights_2:float=0.7207
    ConstraintWeights_3:float=0.2406
   ) -> (Pose) {
    alphaBoolBlend {blendOption=Linear}
    boneToModify {boneName=dyn_butt_flap_r}
    constraintSetup [{targetBone {boneName=spine_03}}, {targetBone {boneName=spine_02}}, {targetBone {boneName=pelvis}}, {targetBone {boneName=spine_01}}]
    constraintWeights [1,1,1,1]
   }
   n5 = Anim Dynamics - Bone: dyn_butt_flap_l(ComponentPose=@n4, Alpha=1, ExternalForce="0,0,0") {
    alpha=0.9
    angularDampingOverride=0.98
    bGravityOverrideInSimSpace=true
    bOverrideAngularDamping=true
    gravityScale=0.75
    numSolverIterationsPostUpdate=4
    numSolverIterationsPreUpdate=16
    alphaBoolBlend {blendOption=Linear}
    boundBone {boneName=dyn_butt_flap_l}
    componentAppliedLinearAccClamp=<5000 5000 5000>
    componentLinearAccScale=<1 1 1>
    componentLinearVelScale=<1 1 1>
    physicsBodyDefinitions [{
      sphereCollisionRadius=0
      boundBone {boneName=dyn_butt_flap_l}
      boxExtents=<20 40 20>
      constraintSetup {angularTargetAxis=AxisZ, twistAxis=AxisZ, angularLimitsMax=<0 0 0>, angularLimitsMin=<-60 0 0>, angularTarget=<0 0 1>}
      localJointOffset=<0 0 -20>
     }]
   }
   n6 = Anim Dynamics - Bone: dyn_butt_flap_r(ComponentPose=@n5, Alpha=1, ExternalForce="0,0,0") {
    alpha=0.9
    angularDampingOverride=0.99
    bGravityOverrideInSimSpace=true
    bOverrideAngularDamping=true
    gravityScale=0.75
    numSolverIterationsPostUpdate=4
    numSolverIterationsPreUpdate=16
    alphaBoolBlend {blendOption=Linear}
    boundBone {boneName=dyn_butt_flap_r}
    componentAppliedLinearAccClamp=<5000 5000 5000>
    componentLinearAccScale=<1 1 1>
    componentLinearVelScale=<1 1 1>
    physicsBodyDefinitions [{
      sphereCollisionRadius=0
      boundBone {boneName=dyn_butt_flap_r}
      boxExtents=<20 40 20>
      constraintSetup {angularTargetAxis=AxisZ, twistAxis=AxisZ, angularLimitsMax=<0 0 0>, angularLimitsMin=<-60 0 0>, angularTarget=<0 0 1>}
      localJointOffset=<0 0 -20>
     }]
   }
   n7 = Anim Dynamics - Bone: dyn_scarf_l(ComponentPose=@n6, Alpha=1, ExternalForce="0,0,0") {
    bGravityOverrideInSimSpace=true
    bLinearSpring=true
    bOverrideLinearDamping=true
    gravityScale=0
    linearDampingOverride=0.8
    linearSpringConstant=100
    alphaBoolBlend {blendOption=Linear}
    boundBone {boneName=dyn_scarf_l}
    componentAppliedLinearAccClamp=<5000 5000 5000>
    physicsBodyDefinitions [{
      sphereCollisionRadius=0
      boundBone {boneName=dyn_scarf_l}
      boxExtents=<10 10 5>
      constraintSetup {angularLimitsMax=<0 0 0>, angularLimitsMin=<0 0 0>, angularTarget=<0 0 0>, linearAxesMax=<2 2 2>, linearAxesMin=<-2 -2 -2>}
      localJointOffset=<0 0 0>
     }]
   }
   n8 = Anim Dynamics - Bone: dyn_scarf_r(ComponentPose=@n7, Alpha=1, ExternalForce="0,0,0") {
    bGravityOverrideInSimSpace=true
    bLinearSpring=true
    gravityScale=0
    linearSpringConstant=100
    alphaBoolBlend {blendOption=Linear}
    boundBone {boneName=dyn_scarf_r}
    componentAppliedLinearAccClamp=<5000 5000 5000>
    physicsBodyDefinitions [{
      sphereCollisionRadius=0
      boundBone {boneName=dyn_scarf_r}
      boxExtents=<10 10 5>
      constraintSetup {angularLimitsMax=<0 0 0>, angularLimitsMin=<0 0 0>, angularTarget=<0 0 0>, linearAxesMax=<2 2 2>, linearAxesMin=<-2 -2 -2>}
      localJointOffset=<0 0 0>
     }]
   }
   n9 = Anim Dynamics - Bone: dyn_scarf_c(ComponentPose=@n8, Alpha=1, ExternalForce="0,0,0") {
    bGravityOverrideInSimSpace=true
    bLinearSpring=true
    gravityScale=0.5
    linearSpringConstant=300
    alphaBoolBlend {blendOption=Linear}
    boundBone {boneName=dyn_scarf_c}
    componentAppliedLinearAccClamp=<5000 5000 5000>
    physicsBodyDefinitions [{
      sphereCollisionRadius=0
      boundBone {boneName=dyn_scarf_c}
      boxExtents=<5 5 5>
      constraintSetup {angularLimitsMax=<0 0 0>, angularLimitsMin=<0 0 0>, angularTarget=<1 0 0>, linearAxesMax=<5 5 5>, linearAxesMin=<-5 -5 -5>}
      localJointOffset=<0 0 0>
     }]
   }
   Save cached pose 'PreScale Cache'(Pose=Component To Local(ComponentPose=@n9)) {cacheName=PreScale Cache}
  ]}
 EventGraph {nodes [
   Event BlueprintUpdateAnimation()
   > Set movementAnimDyn(movementAnimDyn:double=MapRangeClamped(Value=Get PawnSpeed2D(), InRangeA=0, InRangeB=100, OutRangeA=0.1)) -> (Output_Get:double)
   Event BlueprintInitializeAnimation()
   > Set GetPlayerListUsesCustomCharacterParts(GetPlayerListUsesCustomCharacterParts:bool=GetPlaylistUsesCustomCharacterParts!() -> (ReturnValue:bool)) -> (Output_Get:bool)
  ]}
]
variables [
 GetPlayerListUsesCustomCharacterParts {type=boolean}
 movementAnimDyn {type=number}
]
